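# Niagara burst effect for deploying the Chapter 6 Mini Season 2 umbrella glider.
# Flash glow, speed-line/ring, multicolored sparks, gold sparkles and tumbling spring and cog meshes fire on a shared system burst delay, while smoke uses a fixed spawn time.
# All emitters position from a system socket reader.
# Front-end modifiers adjust sizes and offsets, time-of-day multipliers tint colors, and particle partitioning selects spring or cog mesh and scale.
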NiagaraSystem /BRCosmetics/Gadgets/Assets/VinderTech_GliderChute/Glider_Ch_6_Mini_Season_2/FX/NS_Umbrella_Ch6S2_Deploy
dependencies {
 usedDataInterfaces [
  &NiagaraDataInterfaceCurve
  &NiagaraDataInterfaceMeshRendererInfo
  &NiagaraDataInterfaceArrayColor
  &NiagaraDataInterfaceColorCurve
  &NiagaraDataInterfaceSocketReader
 ]
 usedDynamicInputs []
 usedModules [
  &self:SystemSpawnScript.SystemScriptSource.SystemScriptGraph.NiagaraNodeAssignment_0.SetVariables_#1
  &self:SystemSpawnScript.SystemScriptSource.SystemScriptGraph.NiagaraNodeAssignment_4.SetVariables_#2
  &self:SystemSpawnScript.SystemScriptSource.SystemScriptGraph.NiagaraNodeAssignment_5.SetVariables_#3
  &SystemState
  &self:SystemSpawnScript.SystemScriptSource.SystemScriptGraph.NiagaraNodeAssignment_9.SetVariables_#4
 ]
}
emitters {
 Colored_Sparks {
  emitterUpdateScript [
   &EmitterState()
   &SpawnBurst_Instantaneous(
    Spawn Count=&RandomRangeInt(Minimum=3, Maximum=5)
    Spawn Time=$System.BurstDelay
    Use Spawn Probability=false
    UseLoopCountLimit=false
   )
   &SpawnBurst_Instantaneous(
    Spawn Count=&RandomRangeInt(Minimum=8, Maximum=10)
    Spawn Time=&Add_Float(A=0.05, B=$System.BurstDelay)
    Use Spawn Probability=false
    UseLoopCountLimit=false
   ) {module=SpawnBurst_Instantaneous001}
  ]
  particleSpawnScript [
   /SetVariables(Particles.Random=&RandomRangeFloat(Minimum=0, Maximum=1)) {isSetParameters=true}
   &InitializeParticle(
    Color=&MultiplyLinearColor(
     Linear Color=&SelectLinearColorFromArray(
      Array Sampling Mode=::NewEnumerator0
      Direct Array Index=?
      Color Selection Array=DataInterface() {propertyValues {ColorData [<rgba 0 0.2377 0.2341 1>, <rgba 0.2514 0.1265 0.6514 1>, <rgba 0.1563 0.4735 0 1>, <rgba 1 0.664 0 1>, <rgba 0.552 0.0024 0.008 1>]}}
      Random Seed=0
     )
     Scale Factor=&MakeVector4_FromLinearColor(LinearColor=$System.TODColorMultiplier)
    )
    Lifetime=&Lerp_Float(A=0.2, B=0.4, Alpha=$Particles.Random)
    Mesh Scale=&Multiply_VectorByFloat(
     Vector=&Multiply_Vector(A=&Lerp_Vector(A=<0.2 0.2 0.2>, B=<0.6 0.6 0.6>, Alpha=$Particles.Random), B=<0.8 0.8 0.4>)
     Float=&FN_FrontEndFloatModifier(Gameplay=1.5, Front End=1.2)
    )
    Mesh Scale Mode=::NewEnumerator4
    Sprite Size=&Multiply_Vector2D_ByFloat(
     Float=&FN_FrontEndFloatModifier(Gameplay=1.5, Front End=1.2)
     Vector2D=&RandomRangeVector2D(Minimum=<5 40>, Maximum=<10 65>)
    )
    Sprite Size Mode=::NewEnumerator4
    UsePositionOffset=false
    Write Material Random=true
   ) {module=InitializeParticle001}
   &CameraOffset(Camera Offset Amount=&FN_FrontEndFloatModifier(Gameplay=10, Front End=0))
   &SphereLocation(
    MaskSpawnGroup=false
    Override Seed=false
    Sphere Radius=15
    Surface Only=false
    UseNonUniformScale=false
    UseOffset=false
    UseSphereOrientation=false
   )
   &AddVelocity(Velocity=$Output.SphereLocation.SphereNormal, Velocity Speed Scale=&Lerp_Float(A=2, B=3, Alpha=$Particles.Random))
   &DynamicMaterialParameters(
    Index 0 Param 2=&FloatFromCurve(
     FloatCurve=DataInterface() {propertyValues {Curve {keys [<at 0 0 RCIM_Cubic>, {arriveTangent=1, interpMode=RCIM_Cubic, leaveTangent=1, time=0.7035, value=0.3698}, <at 1 1>]}}}
     CurveIndex=&RandomRangeFloat(Minimum=0, Maximum=1)
    )
    Module.Index0Float0Write=true
    Module.Index0Float1Write=true
    Module.Index0Float2Write=true
    Module.Index0Float3Write=true
   ) {module=DynamicMaterialParameters001}
   &InitialMeshOrientation(
    Mesh Orientation Mode=::NewEnumerator2
    Orientation Vector=<1 1 1>
    Rotation=&RandomRangeVector(Minimum=<-1 -1 -1>, Maximum=<1 1 1>)
    Use Rotation Vector=true
   )
  ]
  particleUpdateScript [
   &ParticleState()
   &SocketLocation(
    SamplePosOffset=true
    Sampled Position Offset=<0 0 30>
    Socket Index=1
    Socket Reader=$System.SocketReader
    Socket Sampling Mode=::NewEnumerator3
    VelocitySpeedLimit=false
   )
   &SolveForcesAndVelocity(Clamp Velocity=false, Limit Acceleration=false)
   &DynamicMaterialParameters(
    Index 0 Param 1=&FloatFromCurve(FloatCurve=DataInterface() {propertyValues {Curve {keys [<at 0 0>,<at 1 1>]}}})
    Module.Index0Float0Write=true
    Module.Index0Float1Write=true
    Module.Index0Float2Write=true
    Module.Index0Float3Write=true
   )
   &ScaleColor(
    Scale RGB=&VectorFromFloat(Value=&FN_Select_Float_from_Time_Of_Day(Morning=1, Day=1, Evening=1, Night=0.5))
    ScaleA=false
    ScaleRGB=true
   )
  ]
  rendererClasses [&NiagaraSpriteRendererProperties, &NiagaraMeshRendererProperties]
  renderers [
   {rendererClass=&NiagaraSpriteRendererProperties}
   {rendererClass=&NiagaraMeshRendererProperties, rendererIndex=1}
  ]
 }
 DeployFlash_Glow {
  emitterUpdateScript [
   &EmitterState()
   &SpawnBurst_Instantaneous(Spawn Time=$System.BurstDelay, Use Spawn Probability=false, UseLoopCountLimit=false)
  ]
  particleSpawnScript [
   &InitializeParticle(
    Color=<rgba 1 1 1 0.3>
    Lifetime=0.3
    Mass Mode=::NewEnumerator1
    Position Mode=::NewEnumerator0
    Sprite Rotation Mode=::NewEnumerator2
    Sprite Size Mode=::NewEnumerator3
    Uniform Sprite Size=&Multiply_Float(A=250, B=&FN_FrontEndFloatModifier(Gameplay=1, Front End=$System.FrontEndScale))
    Write Material Random=true
   ) {module=InitializeParticle001}
   &DynamicMaterialParameters(
    Index 0 Param 1=2
    Module.Index0Float0Write=true
    Module.Index0Float1Write=true
    Module.Index0Float2Write=true
    Module.Index0Float3Write=true
   )
   &CameraOffset(Camera Offset Amount=20)
  ]
  particleUpdateScript [
   &ParticleState()
   &SocketLocation(
    SamplePosOffset=true
    Sampled Position Offset=<0 0 30>
    Socket Index=1
    Socket Reader=$System.SocketReader
    Socket Sampling Mode=::NewEnumerator3
    VelocitySpeedLimit=false
   )
   &SolveForcesAndVelocity(Clamp Velocity=false, Limit Acceleration=false)
   &ScaleSpriteSize(
    Uniform Curve Scale=2
    Uniform Curve Sprite Scale=DataInterface() {propertyValues {Curve {keys [
        {arriveTangent=0.7585, interpMode=RCIM_Cubic, leaveTangent=0.7585, tangentMode=RCTM_User, time=0, value=0.4}
        {interpMode=RCIM_Cubic, tangentMode=RCTM_User, time=1, value=1}
       ]}}}
   )
   &ScaleColor(
    Scale Alpha=&Multiply_Float(
     A=&MakeFloatFromLinearColor(LinearColor=$System.TODColorMultiplier, Channel:ENiagaraFloatFromLinearColorOptions=::NewEnumerator3)
     B=&FloatFromCurve(
      FloatCurve=DataInterface() {propertyValues {Curve {keys [
          {arriveTangent=6.4019, interpMode=RCIM_Cubic, leaveTangent=6.4019, tangentMode=RCTM_User, time=0, value=1}
          {interpMode=RCIM_Cubic, leaveTangent=0.002, time=1, value=0}
         ]}}}
      Scale Curve=0.05
     )
    )
    Scale RGB=&MakeVectorFromLinearColorRGB(Color=&MultiplyLinearColor(
      Linear Color=&ColorFromCurve(ColorCurve=DataInterface() {propertyValues {
         AlphaCurve {keys [<at 0 1 RCIM_Cubic>,<at 1 0 RCIM_Cubic>]}
         BlueCurve {keys [<at 0 0>, {interpMode=RCIM_Cubic, leaveTangent=1.3364, time=1, value=0}]}
         GreenCurve {keys [<at 0 2.6218>, {interpMode=RCIM_Cubic, leaveTangent=-5.7627, time=1, value=0.7512}]}
         RedCurve {keys [<at 0 5>, {interpMode=RCIM_Cubic, leaveTangent=-13.364, time=1, value=1}]}
        }})
      Scale Factor=&MakeVector4_FromLinearColor(LinearColor=$System.TODColorMultiplier)
     ))
    ScaleA=true
    ScaleRGB=true
   )
   &FN_Sprite_Attributes(
    Color=&Multiply_LinearColorByFloat(
     Linear Color=&MultiplyLinearColor(
      Linear Color=<rgba 1.5 1.066 0 0.2>
      Scale Factor=&MakeVector4_FromLinearColor(LinearColor=$System.TODColorMultiplier)
     )
     Float=&Multiply_Float(
      A=&FloatFromCurve(FloatCurve=DataInterface() {propertyValues {Curve {keys [<at 0 0 RCIM_Cubic>,<at 0.25 0.75 RCIM_Cubic>,<at 1 0 RCIM_Cubic>]}}})
      B=&MakeFloatFromLinearColor(LinearColor=$System.TODColorMultiplier, Channel:ENiagaraFloatFromLinearColorOptions=::NewEnumerator3)
     )
    )
    Create Color Attribute=true
    Create SpriteSize Attribute=true
    SpriteSize=&Vector2DFromFloat(Value=&Multiply_Float(
      A=&Multiply_Float(A=220, B=&FN_FrontEndFloatModifier(Gameplay=1, Front End=$System.FrontEndScale))
      B=&FloatFromCurve(FloatCurve=DataInterface() {propertyValues {Curve {keys [{arriveTangent=0.9835, interpMode=RCIM_Cubic, leaveTangent=0.9835, tangentMode=RCTM_User, time=0, value=0.5}, <at 1 1 RCIM_Cubic>]}}})
     ))
    Write to Module Namespace=true
   )
  ]
  rendererClasses [&NiagaraSpriteRendererProperties]
  renderers [
   {rendererClass=&NiagaraSpriteRendererProperties}
   {rendererClass=&NiagaraSpriteRendererProperties, rendererIndex=1}
  ]
 }
 Smoke {
  emitterUpdateScript [
   &EmitterState()
   &SpawnBurst_Instantaneous(Spawn Time=0.85, Use Spawn Probability=false, UseLoopCountLimit=false)
  ]
  particleSpawnScript [&InitializeParticle(
    Color=&MultiplyLinearColor(
     Linear Color=<rgba 1.1 1.1 1.1 1>
     Scale Factor=&MakeVector4_FromLinearColor(LinearColor=$System.TODColorMultiplier)
    )
    Lifetime=0.35
    Sprite Rotation Mode=::NewEnumerator2
    Sprite Size Mode=::NewEnumerator3
    Uniform Sprite Size=&Multiply_Float(A=300, B=&FN_FrontEndFloatModifier(Gameplay=1, Front End=$System.FrontEndScale))
    UsePositionOffset=false
    Write Material Random=true
   )]
  particleUpdateScript [
   &ParticleState()
   &SocketLocation(
    SamplePosOffset=true
    Sampled Position Offset=&FN_FrontEndVector3Modifier(Gameplay=<0 0 30>, Front End=<0 0 0>)
    Socket Index=1
    Socket Reader=$System.SocketReader
    Socket Sampling Mode=::NewEnumerator3
    VelocitySpeedLimit=false
   )
   &ScaleSpriteSize(Uniform Curve Sprite Scale=DataInterface() {propertyValues {Curve {
       defaultValue=0
       keys [
        {arriveTangent=2.0491, interpMode=RCIM_Cubic, leaveTangent=2.0491, tangentMode=RCTM_User, time=0, value=0}
        {interpMode=RCIM_Cubic, tangentMode=RCTM_User, time=1, value=1}
       ]
      }}})
   &DynamicMaterialParameters(
    Index 0 Param 1=&FloatFromCurve(FloatCurve=DataInterface() {propertyValues {Curve {keys [<at 0 0>,<at 1 1>]}}})
    Module.Index0Float0Write=true
    Module.Index0Float1Write=true
    Module.Index0Float2Write=true
    Module.Index0Float3Write=true
   )
  ]
  rendererClasses [&NiagaraSpriteRendererProperties]
  renderers [{rendererClass=&NiagaraSpriteRendererProperties}]
 }
 Sparkles {
  emitterUpdateScript [
   &EmitterState()
   &SpawnBurst_Instantaneous(Spawn Count=16, Spawn Time=$System.BurstDelay, Use Spawn Probability=false)
  ]
  particleSpawnScript [
   &InitializeParticle(
    Color=&MultiplyLinearColor(
     Linear Color=<rgba 5 3.3359 0.1578 1>
     Scale Factor=&MakeVector4_FromLinearColor(LinearColor=$System.TODColorMultiplier)
    )
    Lifetime=&UniformRangedFloat(Minimum=0.4, Maximum=1.2)
    Sprite Rotation Angle=45
    Sprite Rotation Mode=::NewEnumerator0
    Sprite Size Mode=::NewEnumerator3
    Uniform Sprite Size=&Multiply_Float(
     A=&UniformRangedFloat(Minimum=10, Maximum=18)
     B=&FN_FrontEndFloatModifier(Gameplay=1, Front End=$System.FrontEndScale)
    )
    UsePositionOffset=false
    Write Material Random=true
   ) {module=InitializeParticle001}
   &SocketLocation(
    SamplePosOffset=true
    Sampled Position Offset=<0 0 30>
    Socket Index=1
    Socket Reader=$System.SocketReader
    Socket Sampling Mode=::NewEnumerator3
    VelocitySpeedLimit=false
   )
   &ShapeLocation(
    MaskSpawnGroup=false
    Override Seed=false
    Sphere Radius=&FN_FrontEndFloatModifier(Gameplay=60, Front End=80)
    Sphere Surface Distribution=1
    UseOwnerScale=true
   )
   &CameraOffset(Camera Offset Amount=20)
   &AddVelocity(
    Module.Use Falloff Distance=false
    Origin Offset=<0 0 120>
    Override Seed=false
    Velocity Mode=::NewEnumerator1
    Velocity Speed=&RandomRangeFloat(Minimum=200, Maximum=205)
   )
   &ApplyInitialForces()
  ]
  particleUpdateScript [
   &ParticleState()
   &ScaleSpriteSize(Uniform Curve Sprite Scale=DataInterface() {propertyValues {Curve {keys [
        {interpMode=RCIM_Cubic, tangentMode=RCTM_User, time=0, value=0.5718}
        {
         arriveTangent=-3.6492
         arriveTangentWeight=0.1689
         interpMode=RCIM_Cubic
         leaveTangent=-3.6492
         leaveTangentWeight=0.9359
         tangentMode=RCTM_User
         time=0.0814
         value=1
        }
        {tangentMode=RCTM_User, time=1, value=0}
       ]}}}) {module=ScaleSpriteSize001}
   &GravityForce(Gravity=<0 0 200>)
   /Niagara/Modules/Update/Forces/CurlNoiseForce(Noise Frequency=0.02, Noise Quality / Cost=::NewEnumerator0, Noise Strength=500, Pan Noise=false)
   &Drag(Drag=5, Use Linear Drag=true, Use Rotational Drag=true)
   &SolveForcesAndVelocity(Clamp Velocity=false, Limit Acceleration=false)
   &ScaleColor(
    Scale Alpha=&FloatFromCurve(FloatCurve=DataInterface() {propertyValues {Curve {keys [
         <at -0.0092 0.0144 RCIM_Cubic>
         {arriveTangent=-0.0123, interpMode=RCIM_Cubic, leaveTangent=-0.0123, time=0.0967, value=0.9767}
         <at 1.0011 0.002 RCIM_Cubic>
        ]}}})
    ScaleA=true
    ScaleRGB=true
   )
  ]
  rendererClasses [&NiagaraSpriteRendererProperties]
  renderers [{rendererClass=&NiagaraSpriteRendererProperties}]
 }
 SpeedLine_And_Ring {
  emitterUpdateScript [
   &EmitterState()
   &SpawnBurst_Instantaneous(Spawn Time=$System.BurstDelay, Use Spawn Probability=false, UseLoopCountLimit=false)
  ]
  particleSpawnScript [&InitializeParticle(
    Color=&MultiplyLinearColor(
     Linear Color=<rgba 1 1 1 8>
     Scale Factor=&MakeVector4_FromLinearColor(LinearColor=$System.TODColorMultiplier)
    )
    Lifetime=0.4
    Sprite Size Mode=::NewEnumerator3
    Uniform Sprite Size=&Multiply_Float(A=250, B=&FN_FrontEndFloatModifier(Gameplay=1, Front End=$System.FrontEndScale))
    UsePositionOffset=false
    Write Material Random=true
   ) {module=InitializeParticle002}]
  particleUpdateScript [
   &ParticleState()
   &SocketLocation(
    SamplePosOffset=true
    Sampled Position Offset=<0 0 40>
    Socket Index=1
    Socket Reader=$System.SocketReader
    Socket Sampling Mode=::NewEnumerator3
    VelocitySpeedLimit=false
   )
   &SolveForcesAndVelocity(Clamp Velocity=false, Limit Acceleration=false)
   &ScaleSpriteSize(Scale Factor=&Vector2DFromFloat(Value=&FloatFromCurve(FloatCurve=DataInterface() {propertyValues {Curve {keys [{arriveTangent=3.9086, interpMode=RCIM_Cubic, leaveTangent=3.9086, tangentMode=RCTM_User, time=0, value=0}, <at 1 1.4 RCIM_Cubic>]}}})))
   &DynamicMaterialParameters(
    Index 0 Param 1=&FloatFromCurve(FloatCurve=DataInterface() {propertyValues {Curve {keys [<at 0 1 RCIM_Cubic>, {arriveTangent=-2.742, interpMode=RCIM_Cubic, leaveTangent=-2.742, tangentMode=RCTM_User, time=1, value=0}]}}})
    Module.Index0Float0Write=true
    Module.Index0Float1Write=true
    Module.Index0Float2Write=true
    Module.Index0Float3Write=true
   )
   &ScaleColor(
    Scale Alpha=&FloatFromCurve(
     FloatCurve=DataInterface() {propertyValues {Curve {keys [
         {arriveTangent=-13.9062, interpMode=RCIM_Cubic, leaveTangent=-13.9062, tangentMode=RCTM_User, time=0, value=7}
         {arriveTangent=-7, interpMode=RCIM_Cubic, leaveTangent=-7, time=0.2488, value=3}
         <at 1 0 RCIM_Cubic>
        ]}}}
     Scale Curve=2
    )
    ScaleA=true
    ScaleRGB=true
   )
   &CameraOffset()
   &FN_Sprite_Attributes(
    Camera Offset=-20
    Color=&MultiplyLinearColor(
     Linear Color=<rgba 5 0.4948 0 1>
     Scale Factor=&MakeVector4_FromLinearColor(LinearColor=&FNTimeOfDayColor(
       Night Color=<rgba 0.5136 0.7312 1 0.6928>
       Morning Color=<rgba 1 0.6721 0.4687 0.9014>
       Day Color=<rgba 1 1 1 1>
       Evening Color=<rgba 1 0.673 0.4364 1>
      ))
    )
    Create Camera Offset Attribute=true
    Create Color Attribute=true
    Create Sprite Rotation Attribute=true
    Create SpriteSize Attribute=true
    SpriteRotation=90
    SpriteSize=&Vector2DFromFloat(Value=&Multiply_Float(
      A=&Multiply_Float(A=120, B=&FN_FrontEndFloatModifier(Gameplay=1, Front End=$System.FrontEndScale))
      B=&FloatFromCurve(FloatCurve=DataInterface() {propertyValues {Curve {keys [
           <at 0 0 RCIM_Cubic>
           {arriveTangent=1, interpMode=RCIM_Cubic, leaveTangent=1, time=0.3925, value=0.6955}
           {tangentMode=RCTM_User, time=1, value=1}
          ]}}})
     ))
    Write to Module Namespace=true
   )
   /SetVariables(Particles.Ring_DynPar:Vector4f=&MakeVector4(
     X=&FloatFromCurve(FloatCurve=DataInterface() {propertyValues {Curve {keys [
          {arriveTangent=3.0691, interpMode=RCIM_Cubic, leaveTangent=3.0691, tangentMode=RCTM_User, time=0, value=0}
          {interpMode=RCIM_Cubic, leaveTangent=1.3494, tangentMode=RCTM_Break, time=0.6518, value=0.9}
          {arriveTangent=-0.4494, interpMode=RCIM_Cubic, leaveTangent=-2.1212, tangentMode=RCTM_Break, time=1, value=1}
         ]}}})
     Y=&FloatFromCurve(FloatCurve=DataInterface() {propertyValues {Curve {keys [
          {arriveTangent=1.4411, interpMode=RCIM_Cubic, leaveTangent=1.4411, tangentMode=RCTM_User, time=0.4, value=0}
          {arriveTangent=3.3028, interpMode=RCIM_Cubic, leaveTangent=3.3028, tangentMode=RCTM_User, time=1, value=1}
         ]}}})
     Z=0
     W=0
    )) {isSetParameters=true}
  ]
  rendererClasses [&NiagaraSpriteRendererProperties]
  renderers [
   {rendererClass=&NiagaraSpriteRendererProperties}
   {rendererClass=&NiagaraSpriteRendererProperties, rendererIndex=1}
  ]
 }
 Springs_and_Cogs {
  emitterSpawnScript [/SetVariables(
    Emitter.Outline_EdgeThickness_Cog:NiagaraFloat=&FN_FrontEndFloatModifier(Gameplay=0.25, Front End=0.5)
    Emitter.Outline_EdgeThickness_Spring:NiagaraFloat=&FN_FrontEndFloatModifier(Gameplay=0.12, Front End=0.3)
   ) {isSetParameters=true}]
  emitterUpdateScript [
   &EmitterState()
   &SpawnBurst_Instantaneous(
    Spawn Count=&RandomRangeInt(Minimum=5, Maximum=8)
    Spawn Time=$System.BurstDelay
    Use Spawn Probability=false
    UseLoopCountLimit=false
   )
   /SetVariables(Emitter.ChunksColor:LinearColor=&MultiplyLinearColor(
     Linear Color=<rgba 0.09 0.09 0.09 1>
     Scale Factor=&MakeVector4_FromLinearColor(LinearColor=$System.TODColorMultiplier)
    )) {isSetParameters=true}
  ]
  particleSpawnScript [
   &PartitionParticles()
   /SetVariables(Particles.VisibilityTag=$Particles.PartitionParticles.Partition) {isSetParameters=true}
   &InitializeParticle(
    Lifetime=&RandomRangeFloat(Minimum=0.7, Maximum=1.1)
    Mesh Scale Mode=::NewEnumerator1
    Mesh Uniform Scale Max=1
    Mesh Uniform Scale Min=0.9
    UsePositionOffset=false
    Write Material Random=true
   )
   &InitialMeshOrientation(Rotation=&RandomRangeVector(Minimum=<0 0 0>, Maximum=<1 1 1>), Use Rotation Vector=true)
   &SocketLocation(
    SamplePosOffset=true
    Sampled Position Offset=<0 0 30>
    Socket Index=1
    Socket Reader=$System.SocketReader
    Socket Sampling Mode=::NewEnumerator3
    VelocitySpeedLimit=false
   )
   &ShapeLocation(MaskSpawnGroup=false, Override Seed=false, Sphere Radius=60, UseOwnerScale=true)
   &AddVelocity(
    Module.Use Falloff Distance=false
    Origin Offset=&FN_FrontEndVector3Modifier(Gameplay=<0 0 20>, Front End=<0 0 -50>)
    Origin Offset Coordinate Space=::World
    Override Seed=false
    Velocity Mode=::NewEnumerator1
    Velocity Speed=&RandomRangeFloat(Minimum=350, Maximum=500)
   )
   /SetVariables(Particles.Scale=&VectorFromFloat(Value=&MergeHelper_FloatSwitch(
      0=&Multiply_Float(A=0.5, B=&RandomRangeFloat(Minimum=1.1, Maximum=1.25))
      1=&Multiply_Float(A=3, B=&RandomRangeFloat(Minimum=1, Maximum=1.2))
      2=?
      3=?
      Index=$Particles.PartitionParticles.Partition
      Num=::NewEnumerator0
     ))) {isSetParameters=true}
  ]
  particleUpdateScript [
   &ParticleState()
   &GravityForce()
   &MeshRotationForce(Rotation=<35 0 0>)
   &Drag(Drag=2, Use Linear Drag=true, Use Rotational Drag=true)
   &SolveForcesAndVelocity(Clamp Velocity=false, Limit Acceleration=false)
   &ScaleMeshSize(Scale Factor=&VectorFromFloat(Value=&FloatFromCurve(FloatCurve=DataInterface() {propertyValues {Curve {
         defaultValue=1
         keys [{interpMode=RCIM_Cubic, tangentMode=RCTM_User, time=0, value=1}, {interpMode=RCIM_Cubic, tangentMode=RCTM_User, time=1, value=0}]
        }}})))
  ]
  rendererClasses [&NiagaraMeshRendererProperties]
  renderers [
   {rendererClass=&NiagaraMeshRendererProperties, rendererIndex=0}
   {rendererClass=&NiagaraMeshRendererProperties, rendererIndex=1}
   {rendererClass=&NiagaraMeshRendererProperties, rendererIndex=2}
   {rendererClass=&NiagaraMeshRendererProperties, rendererIndex=3}
  ]
 }
}
summary {
 systemName=NS_Umbrella_Ch6S2_Deploy
 emitters [
  {emitterName=Smoke, rendererClasses [&NiagaraSpriteRendererProperties]}
  {emitterName=Springs_and_Cogs, rendererClasses [&NiagaraMeshRendererProperties]}
  {emitterName=Colored_Sparks, rendererClasses [&NiagaraSpriteRendererProperties, &NiagaraMeshRendererProperties]}
  {emitterName=Sparkles, rendererClasses [&NiagaraSpriteRendererProperties]}
  {emitterName=DeployFlash_Glow, rendererClasses [&NiagaraSpriteRendererProperties]}
  {emitterName=SpeedLine_And_Ring, rendererClasses [&NiagaraSpriteRendererProperties]}
 ]
 userVariables [
  User.bIsFrontEnd {type=NiagaraBool}
  User.bIsFrontEndPreview {type=NiagaraBool}
 ]
}
