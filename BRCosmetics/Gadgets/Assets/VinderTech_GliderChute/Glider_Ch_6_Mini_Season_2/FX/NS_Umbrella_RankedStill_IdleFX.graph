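# Niagara idle effect for a Chapter 6 Mini Season 2 ranked umbrella glider, attached to skeletal mesh bones and sockets.
# Tube_Glow is a pulsing sprite, Corona a two-renderer mesh glow with animated material parameters, and Smoke spawns across five bones by horizontal owner movement.
# Color switches key off ranked progression or rank index.
# Time of day tints all three, while front-end flags and platform checks adjust the glows.
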NiagaraSystem /BRCosmetics/Gadgets/Assets/VinderTech_GliderChute/Glider_Ch_6_Mini_Season_2/FX/NS_Umbrella_RankedStill_IdleFX
dependencies {
 usedDataInterfaces [
  &NiagaraDataInterfacePlatformSet
  &NiagaraDataInterfaceCurve
  &NiagaraDataInterfaceSkeletalMesh
  &NiagaraDataInterfaceColorCurve
  &NiagaraDataInterfaceArrayColor
 ]
 usedDynamicInputs [
  &UniformRangedFloat
  &Vector2DFromFloat
  &FNTimeOfDayColor
  &AddVectorToPosition
  &ColorFromCurve
  &SelectLinearColorFromArray
 ]
 usedModules [
  &self:SystemSpawnScript.SystemScriptSource.SystemScriptGraph.NiagaraNodeAssignment_7.SetVariables_#1
  &SystemState
  &self:SystemSpawnScript.SystemScriptSource.SystemScriptGraph.NiagaraNodeAssignment_1.SetVariables_#2
  &self:SystemSpawnScript.SystemScriptSource.SystemScriptGraph.NiagaraNodeAssignment_0.SetVariables_#3
  &self:SystemSpawnScript.SystemScriptSource.SystemScriptGraph.NiagaraNodeAssignment_2.SetVariables_#4
  &self:SystemSpawnScript.SystemScriptSource.SystemScriptGraph.NiagaraNodeAssignment_4.SetVariables_#5
 ]
}
emitters {
 Corona {
  emitterUpdateScript [
   &EmitterState(
    Inactive Response=::NewEnumerator1
    Life Cycle Mode=::NewEnumerator1
    Loop Behavior=::NewEnumerator1
    Loop Duration Mode=::NewEnumerator1
    UseLoopDelay=false
   )
   &SpawnBurst_Instantaneous(Use Spawn Probability=false, UseLoopCountLimit=false)
   /SetVariables(Emitter.OpacityMultiplier:NiagaraFloat=&Multiply_Float(
     A=&FN_PickFloatDependOnPlatform(TargetPlatformActive=1, TargetPlatformNoActive=0.6)
     B=&FNTimeOfDayScalar(Night Scale=0.7, Morning Scale=0.7, Evening Scale=0.7)
    )) {isSetParameters=true}
  ]
  particleSpawnScript [
   &InitializeParticle(
    Color Mode=::NewEnumerator0
    Mesh Scale=&FN_FrontEndVector3Modifier(Gameplay=<6 6 6>, Front End=<7 7 7>)
    Mesh Scale Mode=::NewEnumerator4
    UsePositionOffset=false
    Write Material Random=true
   )
   &InitialMeshOrientation(Mesh Orientation Mode=::NewEnumerator7, Rotation=<0 0.5 0>, Use Rotation Vector=true)
   &CameraOffset(Camera Offset Amount=&FN_PickFloatDependOnPlatform(TargetPlatformActive=0, TargetPlatformNoActive=-8))
  ]
  particleUpdateScript [
   &SkeletalMeshLocation(
    Bone / Socket Index=5
    Bone Sampling Mode=::NewEnumerator4
    MaskSpawnGroup=false
    Sampled Position Offset=<0 0 -2.5>
    Skeletal Mesh=$System.Skeletal Mesh Reader
   )
   /SetVariables(Particles.InnerColor:Vector4f=&Multiply_Vector4(
     A=&MakeVector4FromVectorAndFloat(
      "Vector (RGB)"=&MakeVectorFromLinearColorRGB(Color=&MergeHelper_ColorSwitch(
        0=<rgba 0 0 0 0>
        1=<rgba 0 0 0 0>
        2=<rgba 0 0 0 0>
        3=<rgba 0 0 0 0>
        4=<rgba 0 0 0 0>
        5=&MakeCustomLinearColorFromBool(
         Bool=&Boolean_OrOperation(Bool1=$User.bIsFrontEnd, Bool2=$User.bIsFrontEndPreview)
         True Color=<rgba 0 1 0.8 1>
         False Color=<rgba 0.3124 1.5 1.2031 1>
        )
        6=&MakeCustomLinearColorFromBool(
         Bool=&Boolean_OrOperation(Bool1=$User.bIsFrontEnd, Bool2=$User.bIsFrontEndPreview)
         True Color=<rgba 1 0.6933 0.08 1>
         False Color=<rgba 5 0.8261 0 1>
        )
        7=<rgba 0.875 0.5 5 1>
        Index=$User.RankedProgression
        Num=::NewEnumerator6
       ))
      "Float (Alpha)"=&Multiply_Float(A=$System.RevealCurve, B=$Emitter.OpacityMultiplier)
     )
     B=&MakeVector4_FromLinearColor(LinearColor=$System.TODColorMultiplier)
    )) {isSetParameters=true}
   &ParticleState(Kill Particles When Lifetime Has Elapsed=false, Let Infinitely Lived Particles Die When Emitter Deactivates=true)
   &DynamicMaterialParameters(
    Index 0 Param 1=&RandomRangeFloat(Minimum=0, Maximum=1)
    Index 0 Param 2=1
    Index 0 Param 3=3
    Index 0 Param 4=3
    Index 2 Param 1=&RandomRangeFloat(Minimum=0, Maximum=1)
    Index 2 Param 2=-2
    Index 2 Param 3=&Lerp_Float(A=5, B=15, Alpha=&Waveform("Waveform [1]"=::NewEnumerator12, "[1] Frequency"=3, "[1] Amplitude Min/Max"=<0 1>))
    Index 2 Param 4=&Lerp_Float(A=3, B=12, Alpha=&Waveform("Waveform [1]"=::NewEnumerator12, "[1] Frequency"=1, "[1] Amplitude Min/Max"=<0 1>))
    Module.Index0Float0Write=true
    Module.Index0Float1Write=true
    Module.Index0Float2Write=true
    Module.Index0Float3Write=true
    Module.Index2Float0Write=true
    Module.Index2Float1Write=true
    Module.Index2Float2Write=true
    Module.Index2Float3Write=true
    Write Parameter Index 2=::NewEnumerator1
   )
   /SetVariables(Particles.Color=&MultiplyLinearColor(
     Linear Color=&MultiplyLinearColor(
      Linear Color=&MergeHelper_ColorSwitch(
       0=<rgba 0 0 0 0>
       1=&MakeCustomLinearColorFromBool(
        Bool=&Boolean_OrOperation(Bool1=$User.bIsFrontEnd, Bool2=$User.bIsFrontEndPreview)
        True Color=<rgba 0.0609 1 0.7652 1>
        False Color=<rgba 0.1458 0.7 0.5614 1>
       )
       2=&MakeCustomLinearColorFromBool(
        Bool=&Boolean_OrOperation(Bool1=$User.bIsFrontEnd, Bool2=$User.bIsFrontEndPreview)
        True Color=<rgba 2 0.6 0 1>
        False Color=<rgba 0.7 0.1157 0 3>
       )
       3=<rgba 0 0.8333 10 1>
       Index=/Niagara/DynamicInputs/Subtract/Subtract_Integer(A:NiagaraInt32=$User.RankedProgression, B:NiagaraInt32=15)
       Num=::NewEnumerator2
      )
      Scale Factor=&MakeVector4_FromFloat(Float=&Multiply_Float(A=$System.RevealCurve, B=$Emitter.OpacityMultiplier))
     )
     Scale Factor=&MakeVector4_FromLinearColor(LinearColor=$System.TODColorMultiplier)
    )) {isSetParameters=true}
   &FN_Mesh_Attributes(
    Color=&MultiplyLinearColor(
     Linear Color=&MultiplyLinearColor(
      Linear Color=&MergeHelper_ColorSwitch(
       0=<rgba 0 0 0 0>
       1=&MakeCustomLinearColorFromBool(
        Bool=&Boolean_OrOperation(Bool1=$User.bIsFrontEnd, Bool2=$User.bIsFrontEndPreview)
        True Color=<rgba 0 0.4 0.32 1>
        False Color=<rgba 0 1 0.8 1>
       )
       2=&MakeCustomLinearColorFromBool(
        Bool=&Boolean_OrOperation(Bool1=$User.bIsFrontEnd, Bool2=$User.bIsFrontEndPreview)
        True Color=<rgba 1.25 0.3125 0 1>
        False Color=<rgba 1 0.1651 0 1>
       )
       3=&MakeCustomLinearColorFromBool(
        Bool=&Boolean_OrOperation(Bool1=$User.bIsFrontEnd, Bool2=$User.bIsFrontEndPreview)
        True Color=<rgba 0 0 3 1>
        False Color=<rgba 0 0 3 1>
       )
       Index=&Subtract_Integer(A:NiagaraInt32=$User.RankedProgression, B:NiagaraInt32=15)
       Num=::NewEnumerator2
      )
      Scale Factor=&MakeVector4_FromFloat(Float=&Multiply_Float(A=$System.RevealCurve, B=$Emitter.OpacityMultiplier))
     )
     Scale Factor=&MakeVector4_FromLinearColor(LinearColor=$System.TODColorMultiplier)
    )
    Create Color Attribute=true
    Create Scale Attribute=true
    Scale=<5.5 7 5.5>
    Write to Module Namespace=true
   )
   /SetVariables(Particles.DynPar_CoronaOuter:Vector4f=&MakeVector4(X=&RandomRangeFloat(Minimum=0, Maximum=1), Y=5, Z=3, W=0)) {isSetParameters=true}
  ]
  rendererClasses [&NiagaraMeshRendererProperties]
  renderers [
   {rendererClass=&NiagaraMeshRendererProperties, rendererIndex=0}
   {rendererClass=&NiagaraMeshRendererProperties, rendererIndex=1}
  ]
 }
 Smoke {
  emitterUpdateScript [
   &EmitterState(
    Delay First Loop Only=true
    Inactive Response=::NewEnumerator1
    Life Cycle Mode=::NewEnumerator1
    Loop Delay=2
    Loop Duration=1
    Recalculate Duration Each Loop=true
    UseLoopDelay=true
   )
   &SpawnRate(SpawnRate=50, Use Spawn Probability=false) {enabled=false}
   &SpawnPerUnit(
    Module.Use Max Movement Threshold=true
    Module.Use Movement Tolerance=true
    Spawn Spacing=30
    Use Spawn Probability=false
    Velocity Vector:Vector3f=&Multiply_Vector(A=$Engine.Owner.Velocity, B=<1 1 0>)
   )
  ]
  particleSpawnScript [
   &PartitionParticles(Number Of Partitions=5)
   &InitializeParticle(
    Lifetime=0.45
    Sprite Rotation Mode=::NewEnumerator2
    Sprite Size Mode=::NewEnumerator1
    Uniform Sprite Size Max=7
    UsePositionOffset=false
    Write Material Random=true
   )
   &SkeletalMeshLocation(
    Bone / Socket Index=$Particles.PartitionParticles.Partition
    Bone Sampling Mode=::NewEnumerator4
    MaskSpawnGroup=false
    Orientation Sampling=::NewEnumerator1
    Skeletal Mesh=$System.Skeletal Mesh Reader
   )
   &AddVelocity(Velocity=<0 0 15>)
  ]
  particleUpdateScript [
   &ParticleState()
   &SkeletalMeshLocation(
    Bone / Socket Index=$Particles.PartitionParticles.Partition
    Bone Sampling Mode=::NewEnumerator4
    Initialize Particle Position To Sampled Position=false
    MaskSpawnGroup=false
    Orientation Sampling=::NewEnumerator1
    Position Sampling=::NewEnumerator1
    Skeletal Mesh=$System.Skeletal Mesh Reader
    Source Movement Scale:Vector3f=&VectorFromFloat(Value=0.98)
   ) {module=SkeletalMeshLocation001}
   &SolveForcesAndVelocity(Clamp Velocity=false, Limit Acceleration=false)
   &ScaleSpriteSize(Uniform Curve Sprite Scale=DataInterface() {propertyValues {Curve {keys [
        {arriveTangent=3.1494, interpMode=RCIM_Cubic, leaveTangent=3.1494, tangentMode=RCTM_User, time=0, value=1}
        {tangentMode=RCTM_User, time=1.0011, value=2.5}
       ]}}})
   &CameraOffset()
   &DynamicMaterialParameters(
    Index 0 Param 1=&FloatFromCurve(FloatCurve=DataInterface() {propertyValues {Curve {defaultValue=1, keys [{tangentMode=RCTM_User, time=0, value=0}, {tangentMode=RCTM_User, time=1, value=1}]}}})
    Index 0 Param 2=&RandomRangeFloat(Minimum=0, Maximum=2)
    Index 0 Param 3=1
    Index 0 Param 4=1
    Index 1 Param 1=&FloatFromCurve(FloatCurve=DataInterface() {propertyValues {Curve {keys [
         {arriveTangent=-2.3163, interpMode=RCIM_Cubic, time=0, value=2}
         {arriveTangent=-1.9019, interpMode=RCIM_Cubic, leaveTangent=-1.9019, tangentMode=RCTM_User, time=1, value=0.4}
        ]}}})
    Index 1 Param 2=0.4
    Index 1 Param 3=0.6
    Index 1 Param 4=1
    Index 2 Param 1=&FloatFromCurve(FloatCurve=DataInterface() {propertyValues {Curve {keys [<at 0.46 0>,<at 1 1>]}}})
    Index 2 Param 2=45
    Module.Index0Float0Write=true
    Module.Index0Float1Write=true
    Module.Index0Float2Write=true
    Module.Index0Float3Write=true
    Module.Index1Float0Write=true
    Module.Index1Float1Write=true
    Module.Index1Float2Write=true
    Module.Index1Float3Write=true
    Module.Index2Float0Write=true
    Module.Index2Float1Write=true
    Module.Index2Float2Write=true
    Module.Index2Float3Write=true
    Write Parameter Index 1=::NewEnumerator1
    Write Parameter Index 2=::NewEnumerator1
   ) {module=DynamicMaterialParameters001}
   &ScaleColor(
    Scale Alpha=&Multiply_Float(
     A=&MakeFloatFromLinearColor(LinearColor=$System.TODColorMultiplier, Channel:ENiagaraFloatFromLinearColorOptions=::NewEnumerator3)
     B=&FloatFromCurve(
      FloatCurve=DataInterface() {propertyValues {Curve {keys [
          {arriveTangent=2.2572, interpMode=RCIM_Cubic, leaveTangent=2.2572, tangentMode=RCTM_User, time=0, value=0.25}
          {
           arriveTangent=0.0254
           arriveTangentWeight=0.4125
           interpMode=RCIM_Cubic
           leaveTangentWeight=0.1667
           tangentMode=RCTM_Break
           time=0.5
           value=1
          }
          {arriveTangent=-1.8278, tangentMode=RCTM_User, time=1, value=0}
         ]}}}
      Scale Curve=&FNTimeOfDayScalar(Night Scale=0.2, Morning Scale=0.25, Day Scale=0.25, Evening Scale=0.25)
     )
    )
    Scale RGB=&MakeVectorFromLinearColorRGB(Color=&MultiplyLinearColor(
      Linear Color=&MergeHelper_ColorSwitch(
       0=<rgba 1.5 0.8853 0.3 1>
       1=<rgba 0.8006 1.5 1.462 1>
       2=<rgba 1.5 1.15 0.45 1>
       3=<rgba 0.6 1.5 1.2837 1>
       4=<rgba 0.64 1.12 1.6 1>
       5=<rgba 0.48 1.2 1.2 1>
       6=<rgba 1.5 0.6724 0.3 1>
       7=<rgba 0.645 0.6 1.5 1>
       Index=$System.RankIndex
       Num=::NewEnumerator6
      )
      Scale Factor=&MakeVector4_FromLinearColor(LinearColor=$System.TODColorMultiplier)
     ))
    ScaleA=true
    ScaleRGB=true
   ) {module=ScaleColor001}
  ]
  rendererClasses [&NiagaraSpriteRendererProperties]
  renderers [{rendererClass=&NiagaraSpriteRendererProperties}]
 }
 Tube_Glow {
  emitterUpdateScript [
   &EmitterState(
    Inactive Response=::NewEnumerator1
    Life Cycle Mode=::NewEnumerator1
    Loop Behavior=::NewEnumerator1
    Loop Duration Mode=::NewEnumerator1
    UseLoopDelay=false
   )
   &SpawnBurst_Instantaneous(Use Spawn Probability=false, UseLoopCountLimit=false)
  ]
  particleSpawnScript [
   &InitializeParticle(
    Sprite Size=&FN_FrontEndVector2Modifier(Front End:Vector2f=<25 50>, Gameplay:Vector2f=<30 60>)
    Sprite Size Mode=::NewEnumerator4
    UsePositionOffset=false
    Write Material Random=true
   )
   &CameraOffset(Camera Offset Amount=&FN_PickFloatDependOnPlatform(TargetPlatformActive=4, TargetPlatformNoActive=-4))
  ]
  particleUpdateScript [
   &ParticleState(Kill Particles When Lifetime Has Elapsed=false, Let Infinitely Lived Particles Die When Emitter Deactivates=true)
   &SkeletalMeshLocation(
    MaskSpawnGroup=false
    Mesh Sampling Type=::NewEnumerator5
    Sampled Position Offset=&FN_FrontEndVector3Modifier(Gameplay=<0 0 2>, Front End=<0 0 2>)
    Skeletal Mesh=$System.Skeletal Mesh Reader
    Socket Sampling Mode=::NewEnumerator5
   )
   &SpriteFacingAndAlignment(Sprite Alignment=<0 0 1>, Sprite Facing=<0 1 0>, WriteAlignment=true, WriteFacing=false)
   &ScaleColor(
    Scale Alpha=&Multiply_Float(
     A=&Multiply_Float(
      A=&Lerp_Float(
       A=0
       B=&FN_FrontEndFloatModifier(Gameplay=0.5, Front End=0.25)
       Alpha=&Add_Float(
        A=&Multiply_Float(
         A=&Frac_Float(Float=&Sine(Normalized Angle=$Engine.Time, Period=1, Scale=1, Bias=0))
         B=0.5
        )
        B=0.5
       )
      )
      B=&FloatFromCurve(
       FloatCurve=DataInterface() {propertyValues {Curve {defaultValue=1, keys [{tangentMode=RCTM_User, time=0, value=0}, {tangentMode=RCTM_User, time=0.5, value=1}]}}}
       CurveIndex=$Emitter.Age
      )
     )
     B=&Multiply_Float(
      A=&FN_PickFloatDependOnPlatform(TargetPlatformActive=1, TargetPlatformNoActive=0.7)
      B=&FNTimeOfDayScalar(
       Night Scale=0.8
       Morning Scale=0.7
       Day Scale=&FN_FrontEndFloatModifier(Gameplay=1, Front End=2)
       Evening Scale=0.7
      )
     )
    )
    Scale RGB=&MakeVectorFromLinearColorRGB(Color=&MultiplyLinearColor(
      Linear Color=&MergeHelper_ColorSwitch(
       0=<rgba 1 0.4267 0.2 1>
       1=<rgba 0.3198 0.6441 0.7 1>
       2=<rgba 0.8 0.68 0.32 1>
       3=<rgba 0.5 1 0.8798 1>
       4=<rgba 0.35 0.675 1 1>
       5=<rgba 0.5 1 1 1>
       6=<rgba 1.5 0.4425 0.15 1>
       7=<rgba 0.375 0.7396 1.25 1>
       Index=$User.RankedProgression
       Num=::NewEnumerator6
      )
      Scale Factor=&MakeVector4_FromLinearColor(LinearColor=$System.TODColorMultiplier)
     ))
    ScaleA=true
    ScaleRGB=true
   )
   &DynamicMaterialParameters(
    Index 0 Param 1=1.5
    Index 0 Param 2=20
    Index 0 Param 3=1
    Index 0 Param 4=1
    Module.Index0Float0Write=true
    Module.Index0Float1Write=true
    Module.Index0Float2Write=true
    Module.Index0Float3Write=true
   )
  ]
  rendererClasses [&NiagaraSpriteRendererProperties]
  renderers [
   {rendererClass=&NiagaraSpriteRendererProperties}
   {rendererClass=&NiagaraSpriteRendererProperties, rendererIndex=1}
  ]
 }
}
summary {
 systemName=NS_Umbrella_RankedStill_IdleFX
 emitters [
  {emitterName=Tube_Glow, rendererClasses [&NiagaraSpriteRendererProperties]}
  {emitterName=Smoke, rendererClasses [&NiagaraSpriteRendererProperties]}
  {emitterName=Corona, rendererClasses [&NiagaraMeshRendererProperties]}
 ]
 userVariables [
  User.bIsFrontEnd {type=NiagaraBool}
  User.bIsFrontEndPreview {type=NiagaraBool}
  User.RankedProgression {type=NiagaraInt32}
 ]
}
