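# Level Sequence that stages the deploy Niagara effect for the Chapter 6 Mini Season 2 umbrella glider.
# It binds the glider skeletal mesh actor at a fixed transform, with an animation track that holds no sections.
# The NS_Umbrella_Ch6S2_Deploy Niagara component is bound as a child of that actor and driven by a single system spawn section.
# No master tracks are present.
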LevelSequence /BRCosmetics/Gadgets/Assets/VinderTech_GliderChute/Glider_Ch_6_Mini_Season_2/FX/Sequence_GliderCh6_MiniSeason2_FX
bindings [
 NS_Umbrella_Ch6S2_Deploy {
  id=#1
  parent_id=#2
  possessed_class=NiagaraComponent
  tracks [{type=MovieSceneNiagaraSystemTrack, sections [{type=MovieSceneNiagaraSystemSpawnSection, end_frame=300, start_frame=0}]}]
 }
 Ch_6_Mini_Season_2_Glider {
  id=#2
  possessed_class=FortSkeletalMeshActor
  tracks [
   {
    type=MovieScene3DTransformTrack
    sections [{
      type=MovieScene3DTransformSection
      channels [
       Location.X {type=MovieSceneScriptingDoubleChannel, default=-128}
       Location.Y {type=MovieSceneScriptingDoubleChannel, default=1196}
       Location.Z {type=MovieSceneScriptingDoubleChannel, default=1952}
       Rotation.X {type=MovieSceneScriptingDoubleChannel, default=0}
       Rotation.Y {type=MovieSceneScriptingDoubleChannel, default=0}
       Rotation.Z {type=MovieSceneScriptingDoubleChannel, default=0}
       Scale.X {type=MovieSceneScriptingDoubleChannel, default=1}
       Scale.Y {type=MovieSceneScriptingDoubleChannel, default=1}
       Scale.Z {type=MovieSceneScriptingDoubleChannel, default=1}
      ]
     }]
   }
   {type=MovieSceneSkeletalAnimationTrack, sections []}
  ]
 }
]
display_rate {denominator=1, numerator=30}
tick_resolution {denominator=1, numerator=24000}
tracks []
