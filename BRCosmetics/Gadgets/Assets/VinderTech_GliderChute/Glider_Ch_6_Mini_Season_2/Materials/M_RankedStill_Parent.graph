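# Parent material for the VinderTech glider's ranked "Still" look, built on the 2023 anime character shader functions (Pre, Calculations, Post).
# Lit and shaded colors get levels-adjusted desaturation and a two-color Fresnel tint before entering Post.
# Emissive combines a tinted, time-pulsed map with a noise-modulated Fresnel rank glow.
# Thin-film and metal curve-LUT color branches exist, but Post's color inputs take the tinted colors from before them.
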Material /BRCosmetics/Gadgets/Assets/VinderTech_GliderChute/Glider_Ch_6_Mini_Season_2/Materials/M_RankedStill_Parent
graph {
 properties {
  bAutomaticallySetUsageInEditor=false
  bUseMaterialAttributes=true
  bUsedWithMeshDeformer=true
  bUsedWithMorphTargets=true
  bUsedWithSkeletalMesh=true
  bUsedWithStaticLighting=true
  pixelDepthOffsetMode=PDOM_Legacy
  usedShadingModels=Default Lit
 }
 nodes [
  MaterialExpressionMaterialFunctionCall_21 = MaterialFunctionCall!(CustomUV) {materialFunction=/Game/Global/Materials/Character_Parent_2023/Anime_Parent_2023/MaterialFunctions/MFC_AnimeShader_2023_Pre}
  n1 = Constant() {r=1}
  n2 = Desaturation(None=@MaterialExpressionMaterialFunctionCall_21.Color_Shaded, Fraction=@n1)
  n3 = ScalarParameter() {group=04_Ranked, parameterName=Levels_Black, sliderMax=1}
  n4 = ScalarParameter() {defaultValue=0.5, group=04_Ranked, parameterName=Levels_Mid, sliderMax=1}
  n5 = ScalarParameter() {defaultValue=1, group=04_Ranked, parameterName=Levels_White, sliderMax=1}
  n6 = StaticBool() {value=true}
  n7 = ScalarParameter() {defaultValue=2, group=04_Ranked, parameterName=Levels_Gamma, sliderMax=4}
  n8 = ScalarParameter() {group=04_Ranked, parameterName=DesaturationAmount, sliderMax=1}
  n9 = VectorParameter() {group=04_Ranked, parameterName=RankedTint_Color, defaultValue=<rgba 1 1 1 0>}
  n10 = Fresnel(ExponentIn=ScalarParameter() {defaultValue=1, group=04_Ranked, parameterName=TwoColorTint_FresnelExp, sliderMax=8}) {baseReflectFraction=0, exponent=1}
  n11 = LinearInterpolate(
   A=@n9.RGB
   B=LinearInterpolate(
    A=@n9.RGB
    B=VectorParameter() {group=04_Ranked, parameterName=RankedTint_Color2, defaultValue=<rgba 1 1 1 0>}.RGB
    Alpha=Saturate(None=@n10)
   )
   Alpha=ScalarParameter() {group=04_Ranked, parameterName=TwoColorTint_Blend, sliderMax=1}
  )
  n12 = Multiply(
   A=LinearInterpolate(
    A=@MaterialExpressionMaterialFunctionCall_21.Color_Shaded
    B=MaterialFunctionCall!(
     Texture=@n2
     ---------------
     New Black Value=@n3
     New Middle Value=@n4
     New White Value=@n5
     Middle Point
     "--------------- "
     Define Interpolation Curve=@n6
     Interpolation Power=@n7
     Invert Interpolation Power
    ) {materialFunction=/Engine/Functions/Engine_MaterialFunctions02/3PointLevels}.Result
    Alpha=@n8
   )
   B=@n11
  ) {desc=CS}
  n13 = StaticSwitchParameter(True=NamedRerouteUsage(), False=NamedRerouteUsage()) {group=06_Thin Film, parameterName=UseColorAdjustment, sortPriority=100}
  n14 = Saturate(None=Power(Base=NamedRerouteUsage(), Exp=ScalarParameter() {defaultValue=2, group=06_Thin Film, parameterName=ThinFilmExponent}))
  n15 = ScalarParameter() {group=04_Ranked, parameterName=ThinFilm_Blend, sliderMax=1}
  n16 = LinearInterpolate(A=@n12, B=LinearInterpolate(A=@n12, B=@n13, Alpha=@n14), Alpha=@n15)
  n17 = CameraVectorWS()
  n18 = PixelNormalWS()
  n19 = CurveAtlasRowParameter(CurveTime=Saturate(None=DotProduct(A=@n17, B=@n18))) {
   atlas=/Game/Global/Textures/Default/Curve/FN_Character_Atlas
   curve=/Game/Characters/Player/Female/Medium/Bodies/F_MED_Ms_Alpine/Textures/AluminiumreflectanceCurve
   group=06_Metal LUT
   parameterName=Metal LUT Curve
  }
  n20 = ChannelMaskParameter(None=@n19) {group=06_Metal LUT, maskChannel=Blue, parameterName=LUTChannel, defaultValue=<rgba 0 0 1 0>}
  n21 = Saturate(None=ScalarParameter() {group=06_Metal LUT, parameterName=MetalLutIntensity, sliderMax=1})
  LinearInterpolate(A=@n16, B=Multiply(A=@n16, B=@n20.R), Alpha=@n21)
  n22 = Desaturation(None=@MaterialExpressionMaterialFunctionCall_21.Color_Lit, Fraction=@n1)
  n23 = Multiply(
   A=LinearInterpolate(
    A=@MaterialExpressionMaterialFunctionCall_21.Color_Lit
    B=MaterialFunctionCall(Texture=@n22, New Black Value=@n3, New Middle Value=@n4, New White Value=@n5, Define Interpolation Curve=@n6, Interpolation Power=@n7) {materialFunction=&3PointLevels}.Result
    Alpha=@n8
   )
   B=@n11
  ) {desc=CL}
  n24 = LinearInterpolate(A=@n23, B=LinearInterpolate(A=@n23, B=@n13, Alpha=@n14), Alpha=@n15)
  LinearInterpolate(A=@n24, B=Multiply(A=@n24, B=@n20.R), Alpha=@n21)
  n25 = MaterialFunctionCall!(
   CustomLightVector
   OverrideNormal_WS
   STT=@MaterialExpressionMaterialFunctionCall_21.STT
   Normal_Map=@MaterialExpressionMaterialFunctionCall_21.STT
   DFL=@MaterialExpressionMaterialFunctionCall_21.STT
   FX=@MaterialExpressionMaterialFunctionCall_21.STT
  ) {materialFunction=/Game/Global/Materials/Character_Parent_2023/Anime_Parent_2023/MaterialFunctions/MFC_AnimeShader_2023_Calculations}
  n26 = VectorParameter() {group=04_Ranked, parameterName=RankedTintEmissive_Color, defaultValue=<rgba 1 1 1 1>}
  n27 = Constant() {r=1}
  n28 = LinearInterpolate(
   A=@MaterialExpressionMaterialFunctionCall_21.Emissive
   B=Multiply(A=Multiply(A=@n26.RGB, B=@n26.RGB), B=Desaturation(None=@MaterialExpressionMaterialFunctionCall_21.Emissive, Fraction=@n27))
   Alpha=ScalarParameter() {group=04_Ranked, parameterName=EmissiveMap_Tint_Blend, sliderMax=1}
  )
  n29 = NamedRerouteUsage()
  n30 = Divide(A=Time(), B=ScalarParameter() {defaultValue=1, group=04_Ranked, parameterName=TimePeriod})
  n31 = Sine(None=Frac(None=Add(A=@n29, B=@n30)))
  n32 = LinearInterpolate(Alpha=Add(A=Multiply(A=@n31) {constB=0.5}) {constB=0.5})
  MaterialFunctionCall!(
   Shading_Gradient=@n25.Shading_Gradient
   Threshold_Gradient=@n25.Shading_Gradient
   Specular_Gradient=@n25.Shading_Gradient
   Sheen_Mask=@n25.Shading_Gradient
   Thickness=@n25.Shading_Gradient
   DFL_Mask=@n25.Shading_Gradient
   FX=@n25.Shading_Gradient
   Metallic_Gradient=@n25.Shading_Gradient
   RimLight_Gradient=@n25.Shading_Gradient
   NormalBuffer=@n25.Shading_Gradient
   --------
   CINE_LightSources=@n25.Shading_Gradient
   CINE_LightColor=@n25.Shading_Gradient
   ---------
   Color_Lit=@n23
   Color_Shaded=@n12
   SRM=@MaterialExpressionMaterialFunctionCall_21.SRM
   Emissive=Add(A=Multiply(A=@n28, B=@n32) {desc=Emissive}, B=NamedRerouteUsage())
   DFL_Color=@MaterialExpressionMaterialFunctionCall_21.SRM
   Specular_Color=@MaterialExpressionMaterialFunctionCall_21.SRM
   RimLight_Color=@MaterialExpressionMaterialFunctionCall_21.SRM
  ) {materialFunction=/Game/Global/Materials/Character_Parent_2023/Anime_Parent_2023/MaterialFunctions/MFC_AnimeShader_2023_Post}
  n33 = Convert(None=@MaterialExpressionMaterialFunctionCall_21.FX) {
   convertInputs [{
     type=Vector3
     expressionInput {
      expression=&self:MaterialExpressionMaterialFunctionCall_21
      inputName=None
      mask=0
      maskA=0
      maskB=0
      maskG=0
      maskR=0
      outputIndex=3
     }
    }]
   convertOutputs [{type=Scalar}, {type=Scalar}, {type=Scalar}]
  }
  NamedRerouteDeclaration(None=@n33) {name=RecolorMask, nodeColor=<rgba 1 0 0 1>}
  NamedRerouteDeclaration(None=@n33) {name=AnimationGradient, nodeColor=<rgba 0 1 0 1>}
  n34 = Multiply(A=TextureCoordinate(), B=ScalarParameter() {defaultValue=1, group=04_Ranked, parameterName=EmissiveNoise_Tiling, sliderMax=2}) {constB=0.5}
  n35 = VectorParameter() {group=04_Ranked, parameterName=NoiseSpeed, defaultValue=<rgba 0.0185 0.0185 0.06 0.051>}
  n36 = TextureSampleParameter2D(UVs=Panner(Speed=AppendVector(A=@n35.R, B=@n35.R)) {speedX=0.2, speedY=0.12}) {
   group=04_Ranked
   parameterName=ThinFilmUVOffsetTexture
   samplerSource=SSM_Wrap_WorldGroupSettings
   texture=/Game/Global/Textures/Default/Noise/3xPerlin_M
  }
  n37 = Multiply(A=Add(A=@n36.G, B=@n36.G)) {constB=0.5}
  n38 = TextureSampleParameter2D(UVs=Panner(Coordinate=Add(A=@n34, B=@n37), Speed=AppendVector(A=@n35.B, B=@n35.B))) {
   group=04_Ranked
   parameterName=ThinFilmUVOffsetTexture
   samplerSource=SSM_Wrap_WorldGroupSettings
   texture=&3xPerlin_M
  }
  n39 = ScalarParameter() {group=04_Ranked, parameterName=NoiseInfluence, sliderMax=1}
  n40 = Multiply(
   A=Fresnel(ExponentIn=ScalarParameter() {defaultValue=5, group=04_Ranked, parameterName=RankedEmissive_FresnelExponent, sliderMax=12}) {baseReflectFraction=0}
   B=LinearInterpolate(B=@n38.R, Alpha=Saturate(None=@n39)) {constA=1}
  )
  n41 = VectorParameter() {group=04_Ranked, parameterName=RankedEmissive_Color, defaultValue=<rgba 1 1 1 0>}
  n42 = Multiply(A=@n41.RGB, B=@n41.RGB)
  NamedRerouteDeclaration(None=Multiply(A=Multiply(A=@n40, B=@n42), B=ScalarParameter() {group=04_Ranked, parameterName=RankedEmissive, sliderMax=1})) {name=EmissiveRank, nodeColor=<rgba 0.2656 1 0 1>}
  n43 = ScalarParameter() {defaultValue=1, group=06_Thin Film, parameterName=ThinFilmScale, sortPriority=5}
  n44 = Power(Base=NamedRerouteUsage(), Exp=ScalarParameter() {defaultValue=0.5, group=06_Thin Film, parameterName=ThinFilmWarp, sortPriority=4})
  n45 = TextureSampleParameter2D(UVs=Add(A=Multiply(A=@n43, B=@n44), B=ScalarParameter() {group=06_Thin Film, parameterName=ThinFilmOffset})) {
   group=06_Thin Film
   parameterName=Thin Film Texture
   sortPriority=1
   texture=/Game/Global/Textures/Default/Thinfilm/T_ThinFilm_Spectrum_COLOR
  }
  NamedRerouteDeclaration(None=Multiply(A=@n45.RGB, B=ScalarParameter() {defaultValue=0.5, group=06_Thin Film, parameterName=ThinFilmIntensity, sortPriority=2})) {name=BaseThinFilm, nodeColor=<rgba 1 0.8672 0 1>}
  n46 = PixelNormalWS()
  n47 = VectorParameter() {group=06_Thin Film, parameterName=SphericalNormalPosOffset, defaultValue=<rgba 0 0 0 1>}
  n48 = MaterialFunctionCall() {materialFunction=&ObjectPivotPoint}
  n49 = Normalize(VectorInput=Subtract(A=WorldPosition().XYZ, B=Add(A=@n47.RGB, B=@n48.Object Pivot Location)))
  n50 = ScalarParameter() {group=06_Thin Film, parameterName=SphericalNormal_Blend_Amt, sliderMax=1}
  NamedRerouteDeclaration(None=Fresnel(Normal=LinearInterpolate(A=@n46, B=@n49, Alpha=@n50)) {baseReflectFraction=0, exponent=1}) {name=Fresnel, nodeColor=<rgba 1 0 0.2578 1>}
  n51 = ScalarParameter() {group=06_Thin Film, parameterName=ColorAdjustHueShift, sortPriority=101}
  n52 = Desaturation(None=NamedRerouteUsage(), Fraction=ScalarParameter() {group=06_Thin Film, parameterName=ColorAdjustDesaturate, sortPriority=102})
  NamedRerouteDeclaration(None=Power(
    Base=MaterialFunctionCall(Hue Shift Percentage=@n51, Texture=@n52) {materialFunction=&HueShift}.Result
    Exp=ScalarParameter() {defaultValue=1, group=06_Thin Film, parameterName=ColorAdjustExponent, sortPriority=103}
   )) {name=ColorAdjustmentThinFilm, nodeColor=<rgba 0.5469 0 1 1>}
 ]
}
